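# Translucent, unlit, two-sided Niagara mesh-particle material for an edge strip effect on the RedPepper pickaxe.
# SplineThicken and a camera offset combine for vertex offset, while particle color feeds emissive adjustments.
# Two panning samples of a flare texture, UV-distorted by scrolling macro cloud noise, merge with Max.
# A SmoothStep along the second UV channel, driven by the dynamic Reveal parameter, wipes that streak mask in.
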Material /BRCosmetics/Weapons/FORT_Melee/Pickaxe_RedPepper/FX/M_FX_RedPepper_EdgeStrip
graph {
 properties {
  bUsedWithNiagaraMeshParticles=true
  blendMode=BLEND_Translucent
  pixelDepthOffsetMode=PDOM_Legacy
  shadingModel=MSM_Unlit
  twoSided=true
  usedShadingModels=Unlit
 }
 nodes [
  n1 = ScalarParameter() {parameterName=StripThicken}
  Add(
   A=MaterialFunctionCall(WidthBase=@n1, WidthTip=@n1) {materialFunction=&SplineThicken}.WorldPositionOffset
   B=MaterialFunctionCall(Offset Amount=ScalarParameter() {parameterName=CameraOffset}) {materialFunction=/Engine/Functions/Engine_MaterialFunctions02/WorldPositionOffset/CameraOffset}.Camera Offset
  )
  MaterialFunctionCall(Emissive=ParticleColor().RGB) {materialFunction=&MF_FX_EmissiveAdjustments}
  n2 = Normalize(VectorInput=Constant3Vector() {constant=<rgba 1 0 0 0>})
  n3 = ComponentMask(None=TextureCoordinate()) {r=true}
  n4 = ScalarParameter() {parameterName=ThickenOrigin}
  Multiply(A=Multiply(A=Transform(None=@n2), B=Add(A=@n3, B=@n4)), B=@n1)
  Multiply()
  n5 = Multiply(A=NamedRerouteUsage(), B=AppendVector(A=ScalarParameter() {parameterName=MacroNoiseStr}, B=Constant()))
  n6 = VectorParameter() {parameterName=UV_Scale, defaultValue=<rgba 0 0 0 1>}
  n7 = Multiply(A=TextureCoordinate() {uTiling=0.5, vTiling=20}, B=Multiply(A=ComponentMask(None=@n6.RGB) {g=true, r=true}, B=@n6.B))
  n8 = VectorParameter() {parameterName=Speed, defaultValue=<rgba 0 0 0 1>}
  n9 = Add(A=@n5, B=Panner(Coordinate=@n7, Speed=ComponentMask(None=@n8.RGB) {g=true, r=true}) {speedX=0.5})
  n10 = AppendVector(A=Constant() {r=1}, B=ScalarParameter() {parameterName=Layer2Scale})
  n11 = VectorParameter() {parameterName=Layer2Speed, defaultValue=<rgba 0 0 0 1>}
  n12 = Add(A=@n5, B=Panner(Coordinate=Multiply(A=@n7, B=@n10), Speed=ComponentMask(None=@n11.RGB) {g=true, r=true}))
  n13 = OneMinus(None=Multiply(A=DynamicParameter() {paramNames [Reveal,Param2,Param3,Param4]}.Reveal) {constB=1.15})
  n14 = ScalarParameter() {parameterName=Reveal_W}
  n15 = TextureCoordinate() {coordinateIndex=1}
  Multiply(
   A=Max(
    A=TextureSampleParameter2D(UVs=@n9) {parameterName=T_EdgeTexture, texture=/Game/Effects/Fort_Effects/Textures/Flares/T_Beam_Flare_Core}.R
    B=TextureSampleParameter2D(UVs=@n12) {parameterName=T_EdgeTexture, texture=&T_Beam_Flare_Core}.R
   )
   B=SmoothStep(Min=@n13, Max=Add(A=@n13, B=@n14), Value=ComponentMask(None=@n15) {g=true})
  )
  n16 = TextureCoordinate()
  n17 = VectorParameter() {group=Macro, parameterName=Macro_Scale, defaultValue=<rgba 0 0 0 1>}
  n18 = Multiply(A=ComponentMask(None=@n17.RGB) {g=true, r=true}, B=@n17.B)
  n19 = VectorParameter() {group=Macro, parameterName=Macro_Speed, defaultValue=<rgba 0 0 0 1>}
  n20 = TextureSampleParameter2D(UVs=Panner(Coordinate=Multiply(A=@n16, B=@n18), Speed=ComponentMask(None=@n19.RGB) {g=true, r=true}) {bFractionalPart=true, speedX=0.5}) {
   parameterName=T_Noise
   samplerType=SAMPLERTYPE_LinearGrayscale
   texture=/Game/Effects/Fort_Effects/Emotes/UltraLight/Texture/T_Cloud_Greyscale
  }
  NamedRerouteDeclaration(None=Subtract(A=Multiply(A=@n20.R) {constB=2})) {name=MacroNoise, nodeColor=<rgba 1 0.4922 0 1>}
  TextureSampleParameter2D(UVs=@n9) {
   parameterName=T_Noise
   samplerType=SAMPLERTYPE_LinearGrayscale
   texture=&T_Cloud_Greyscale
  }
 ]
}
