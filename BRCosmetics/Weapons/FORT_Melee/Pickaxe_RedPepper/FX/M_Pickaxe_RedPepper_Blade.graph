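# Translucent unlit skeletal-mesh material for the RedPepper pickaxe blade, with glowing runes, edge glow and an animated reveal.
# Emissive layers combine a tinted emissive texture, cloud-pulsed runes and a smoothstepped edge mask, then pass through MF_FX_EmissiveAdjustments.
# Panning cloud noise also builds distorted UVs, while sphere masks define an internal light spot and colored reveal band.
# The Reveal parameter drives a noise-distorted smoothstep wipe along UV V.
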Material /BRCosmetics/Weapons/FORT_Melee/Pickaxe_RedPepper/FX/M_Pickaxe_RedPepper_Blade
graph {
 properties {
  bUsedWithMeshDeformer=true
  bUsedWithSkeletalMesh=true
  blendMode=BLEND_Translucent
  pixelDepthOffsetMode=PDOM_Legacy
  shadingModel=MSM_Unlit
  usedShadingModels=Unlit
 }
 nodes [
  n1 = Multiply(
   A=VectorParameter() {parameterName=Color, defaultValue=<rgba 0 0 0 1>}.RGB
   B=TextureSample(UVs=NamedRerouteUsage()) {texture=/BRCosmetics/Weapons/FORT_Melee/Pickaxe_RedPepper/Textures/T_RedPepper_E}.RGB
  )
  n2 = Multiply(A=@n1, B=NamedRerouteUsage())
  n3 = TextureCoordinate()
  n4 = NamedRerouteUsage()
  n5 = TextureSample(UVs=Panner(Coordinate=TextureCoordinate() {uTiling=0.25}) {speedY=0.5}) {samplerType=SAMPLERTYPE_LinearGrayscale, texture=/Game/Effects/Fort_Effects/Emotes/UltraLight/Texture/T_Cloud_Greyscale}
  n6 = LinearInterpolate(A=ScalarParameter() {parameterName=Runes_Phase_Min}, B=ScalarParameter() {parameterName=Runes_Phase_Max}, Alpha=@n5.R)
  n7 = Multiply(
   A=TextureSample(UVs=LinearInterpolate(A=@n3, B=@n4) {constAlpha=0.25}) {texture=/BRCosmetics/Weapons/FORT_Melee/Pickaxe_RedPepper/Textures/T_RedPepper_FX}.G
   B=Multiply(A=Add(A=@n5.R) {constB=0}, B=Clamp(None=@n6) {maxDefault=10}) {constB=4}
  )
  n8 = Add(
   A=Multiply(A=Add(A=@n2, B=@n1), B=VectorParameter() {parameterName=OnlyReds, defaultValue=<rgba 1 0 0 1>}.RGB)
   B=Multiply(A=VectorParameter() {parameterName=RunesColor, defaultValue=<rgba 0 0 0 1>}.RGB, B=@n7)
  )
  n9 = ScalarParameter() {parameterName=edge_p}
  n10 = ScalarParameter() {parameterName=edge_w}
  n11 = NamedRerouteUsage()
  n12 = Multiply(
   A=SmoothStep(
    Min=@n9
    Max=Add(A=@n9, B=@n10)
    Value=TextureSample(UVs=@n11) {texture=&T_RedPepper_FX}.B
   )
   B=VectorParameter() {parameterName=Edge, defaultValue=<rgba 0 0 0 1>}.RGB
  )
  MaterialFunctionCall(Emissive=Add(A=Add(A=@n8, B=@n12), B=NamedRerouteUsage())) {materialFunction=&MF_FX_EmissiveAdjustments}
  n13 = NamedRerouteUsage()
  n14 = ComponentMask(None=TextureCoordinate()) {g=true}
  n15 = ScalarParameter() {parameterName=Reveal}
  n16 = OneMinus(None=@n15)
  n17 = ScalarParameter() {parameterName=reveal_w}
  n18 = ComponentMask(None=TextureCoordinate()) {g=true}
  n19 = Multiply(A=NamedRerouteUsage(), B=Multiply(A=ScalarParameter() {parameterName=RevealNoiseDistortion}, B=@n15))
  Multiply(
   A=Multiply(A=Add(A=@n13, B=@n7), B=LinearInterpolate(Alpha=@n14) {constA=0.7, constB=0.3})
   B=SmoothStep(Min=@n16, Max=Add(A=@n16, B=@n17), Value=Add(A=@n18, B=@n19))
  )
  n20 = TextureCoordinate() {uTiling=2.5, vTiling=0.78}
  NamedRerouteDeclaration(None=TextureSample(UVs=Panner(Coordinate=@n20) {speedY=1}) {samplerType=SAMPLERTYPE_LinearGrayscale, texture=&T_Cloud_Greyscale}.R) {name=MainNoise, nodeColor=<rgba 0.4766 0 1 1>}
  n21 = Subtract(A=Multiply(A=NamedRerouteUsage()) {constB=2})
  n22 = ScalarParameter() {defaultValue=0.075, parameterName=Noise_DistortUV}
  NamedRerouteDeclaration(None=Add(A=TextureCoordinate(), B=Multiply(A=@n21, B=@n22))) {name=DistortedUV, nodeColor=<rgba 1 0 0.6094 1>}
  n23 = VectorParameter() {parameterName=Light, defaultValue=<rgba 0.25 1 0 1>}
  n24 = SphereMask(A=TextureCoordinate(), B=ComponentMask(None=@n23.RGB) {g=true, r=true}, Radius=@n23.B) {attenuationRadius=0.5, hardnessPercent=0}
  NamedRerouteDeclaration(None=Multiply(A=Power(Base=@n24, Exp=@n23.A), B=VectorParameter() {parameterName=InternalLight, defaultValue=<rgba 0 0 0 1>}.RGB)) {name=Illumination, nodeColor=<rgba 0 0.3437 1 1>}
  n25 = ComponentMask(None=TextureCoordinate()) {g=true}
  n26 = Constant() {r=1}
  n27 = LinearInterpolate(A=Add(A=Constant() {r=1}, B=@n26), B=Subtract(B=@n26) {constA=0}, Alpha=ScalarParameter() {parameterName=Reveal}) {constA=1.5, constB=-0.5}
  NamedRerouteDeclaration(None=Multiply(
    A=SphereMask(A=@n25, B=@n27, Radius=@n26) {attenuationRadius=0.5, hardnessPercent=0}
    B=VectorParameter() {parameterName=RevealColor, defaultValue=<rgba 0 0 0 1>}.RGB
   ) {constB=3}) {name=RevealEffect, nodeColor=<rgba 0 1 0.625 1>}
 ]
}
